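# Physics asset for the ArcticIceTalus medium male head cosmetic, setting up physics bodies for its left and right ear bone chains.
# Each ear has three capsule bodies chained by constraints back to a small spherical-sized master body at its base.
# All joints share tight swing limits and an even tighter twist limit, keeping any ear movement subtle with little axial roll.
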PhysicsAsset /BRCosmetics/Characters/Player/Male/Medium/Heads/M_MED_ArcticIceTalus_Head/Meshes/M_MED_ArcticIceTalus_Head_Physics
bodies {
 dyn_ear01_l [{
   shapeType=Capsule
   length=3.4472
   radius=0.8201
   shapeName=dyn_ear01_l_capsule
   center=<1.5171 0 -0.006>
   rotation=<rot -90 179.9956 -179.9968>
  }]
 dyn_ear01_r [{
   shapeType=Capsule
   length=3.4472
   radius=0.8201
   shapeName=dyn_ear01_r_capsule
   center=<1.5171 0 -0.006>
   rotation=<rot -90 -179.9605 179.9604>
  }]
 dyn_ear02_l [{shapeType=Capsule, length=3.4472, radius=0.8201, shapeName=dyn_ear02_l_capsule, center=<1.5171 0 -0.006>, rotation=<rot -90 180 179.9989>}]
 dyn_ear02_r [{shapeType=Capsule, length=3.4472, radius=0.8201, shapeName=dyn_ear02_r_capsule, center=<1.5171 0 -0.006>, rotation=<rot -90 -0.0001 0>}]
 dyn_ear03_l [{shapeType=Capsule, length=3.4472, radius=0.8201, shapeName=dyn_ear03_l_capsule, center=<1.5171 0 -0.006>, rotation=<rot -90 0.0017 -0.0027>}]
 dyn_ear03_r [{
   shapeType=Capsule
   length=3.4472
   radius=0.8201
   shapeName=dyn_ear03_r_capsule
   center=<1.5171 0 -0.006>
   rotation=<rot -90 -179.8419 179.8416>
  }]
 dyn_ear_l_master [{shapeType=Capsule, length=0, radius=0.5, shapeName=dyn_ear_l_master_capsule}]
 dyn_ear_r_master [{shapeType=Capsule, length=0, radius=0.5, shapeName=dyn_ear_r_master_capsule}]
}
constraints [
 {bone1Name=dyn_ear03_l, bone2Name=dyn_ear02_l, swing1LimitDegrees=10, swing2LimitDegrees=10, twistLimitDegrees=2}
 {bone1Name=dyn_ear02_l, bone2Name=dyn_ear01_l, swing1LimitDegrees=10, swing2LimitDegrees=10, twistLimitDegrees=2}
 {bone1Name=dyn_ear01_l, bone2Name=dyn_ear_l_master, swing1LimitDegrees=10, swing2LimitDegrees=10, twistLimitDegrees=2}
 {bone1Name=dyn_ear03_r, bone2Name=dyn_ear02_r, swing1LimitDegrees=10, swing2LimitDegrees=10, twistLimitDegrees=2}
 {bone1Name=dyn_ear02_r, bone2Name=dyn_ear01_r, swing1LimitDegrees=10, swing2LimitDegrees=10, twistLimitDegrees=2}
 {bone1Name=dyn_ear01_r, bone2Name=dyn_ear_r_master, swing1LimitDegrees=10, swing2LimitDegrees=10, twistLimitDegrees=2}
]
